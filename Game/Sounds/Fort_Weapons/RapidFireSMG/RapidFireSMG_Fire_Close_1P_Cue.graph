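# First-person sound cue for close-range firing of the Rapid Fire SMG.
# Mixes a randomly picked, modulated gunshot with a quieter, briefly delayed shell-casing drop borrowed from the Burst Fire SMG pistol sounds.
# Casings get a slight downward pitch variation, and both random pickers are excluded from branch culling.
# Playback uses first-person weapon attenuation, sound class and gunfire concurrency, with virtualization disabled.
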SoundCue /Game/Sounds/Fort_Weapons/RapidFireSMG/RapidFireSMG_Fire_Close_1P_Cue
root=Mixer_1
nodes [
 {id=Delay_0, children [Random_2], properties {delayMax=0.35, delayMin=0.3}}
 {id=Mixer_1, children [Modulator_0,Modulator_1], properties {inputVolume [1,0.45]}}
 {id=Modulator_0, children [Random_1]}
 {id=Modulator_1, children [Delay_0], properties {pitchMax=1, pitchMin=0.9}}
 {
  id=Random_1
  children [WavePlayer_0,WavePlayer_1,WavePlayer_4,WavePlayer_5]
  properties {bShouldExcludeFromBranchCulling=true, weights [1,1,1,1]}
 }
 {
  id=Random_2
  children [WavePlayer_8, WavePlayer_9, WavePlayer_10, WavePlayer_11, WavePlayer_12]
  properties {bShouldExcludeFromBranchCulling=true, weights [1,1,1,1,1]}
 }
 {id=WavePlayer_0, properties {soundWaveAssetPtr=&/RapidFireSMG_Fire_Close_01}}
 {id=WavePlayer_1, properties {soundWaveAssetPtr=&/RapidFireSMG_Fire_Close_02}}
 {id=WavePlayer_10, properties {soundWaveAssetPtr=/Game/Athena/Sounds/Weapons/Pistol_BurstFireSMG/Pistol_BurstFireSMG_Bullet_Casings_Drop_03}}
 {id=WavePlayer_11, properties {soundWaveAssetPtr=/Game/Athena/Sounds/Weapons/Pistol_BurstFireSMG/Pistol_BurstFireSMG_Bullet_Casings_Drop_04}}
 {id=WavePlayer_12, properties {soundWaveAssetPtr=/Game/Athena/Sounds/Weapons/Pistol_BurstFireSMG/Pistol_BurstFireSMG_Bullet_Casings_Drop_05}}
 {id=WavePlayer_4, properties {soundWaveAssetPtr=&/RapidFireSMG_Fire_Close_03}}
 {id=WavePlayer_5, properties {soundWaveAssetPtr=&/RapidFireSMG_Fire_Close_04}}
 {id=WavePlayer_8, properties {soundWaveAssetPtr=/Game/Athena/Sounds/Weapons/Pistol_BurstFireSMG/Pistol_BurstFireSMG_Bullet_Casings_Drop_01}}
 {id=WavePlayer_9, properties {soundWaveAssetPtr=/Game/Athena/Sounds/Weapons/Pistol_BurstFireSMG/Pistol_BurstFireSMG_Bullet_Casings_Drop_02}}
]
properties {
 attenuationSettings=&Fort_Weapon_Projectile_FirstPerson_Athena_Att
 bDebug=true
 maxDistance=2097152
 soundClassObject=&Fort_WeaponRanged_Fire_1P_PSM
 virtualizationMode=Disabled
 volumeMultiplier=0.85
 concurrencyOverrides {bLimitToOwner=true, maxCount=4, resolutionRule=StopOldest}
 concurrencySet [&Weapon_Gunfire_FirstPerson_Conc]
}
