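# Rapid-fire SMG gunfire cue that layers close, medium and distant shot sounds, each with its own distance attenuation.
# The close layer randomizes four SMG waves under the third-person fire sound class.
# Medium (three random, pitched-up waves) and distant (one volume-varied wave) layers reuse Drum Gun audio, mixed under the distant-fire sound class.
# Local and global gunfire concurrency sets govern playback, and the cue's debug flag is on.
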SoundCue /Game/Sounds/Fort_Weapons/RapidFireSMG/RapidFireSMG_Fire_Close_Cue
root=Mixer_1
nodes [
 {
  id=Attenuation_118
  children [Modulator_0]
  properties {attenuationSettings=&Fort_Weapon_Projectile_Close_Athena_Att, attenuationOverrides {stereoSpread=29871.3984}}
 }
 {
  id=Attenuation_119
  children [Modulator_2]
  properties {attenuationSettings=&Fort_Weapon_Projectile_Medium_Mono_Athena_Att}
 }
 {
  id=Attenuation_120
  children [Modulator_1]
  properties {attenuationSettings=&Fort_Weapon_Projectile_Distant_Mono_Athena_Att}
 }
 {id=Mixer_0, children [Attenuation_119, Attenuation_120], properties {inputVolume [0.3,0.35]}}
 {id=Mixer_1, children [SoundClass_77, SoundClass_76], properties {inputVolume [0.5,1]}}
 {id=Modulator_0, children [Random_1], properties {volumeMin=1.05}}
 {id=Modulator_1, children [WavePlayer_23], properties {pitchMax=1, pitchMin=1, volumeMax=1, volumeMin=0.7}}
 {id=Modulator_2, children [Random_5], properties {pitchMax=1.2, pitchMin=1.1, volumeMax=2, volumeMin=2}}
 {
  id=Random_1
  children [WavePlayer_0,WavePlayer_1,WavePlayer_4,WavePlayer_5]
  properties {bShouldExcludeFromBranchCulling=true, weights [1,1,1,1]}
 }
 {id=Random_5, children [WavePlayer_20, WavePlayer_21, WavePlayer_22], properties {bShouldExcludeFromBranchCulling=true, weights [1,1,1]}}
 {id=SoundClass_76, children [Mixer_0], properties {soundClassOverride=&Fort_WeaponRanged_Fire_Distant}}
 {
  id=SoundClass_77
  children [Attenuation_118]
  properties {soundClassOverride=&Fort_WeaponRanged_Fire_3P_PSM}
 }
 {id=WavePlayer_0, properties {soundWaveAssetPtr=&/RapidFireSMG_Fire_Close_01}}
 {id=WavePlayer_1, properties {soundWaveAssetPtr=&/RapidFireSMG_Fire_Close_02}}
 {id=WavePlayer_20, properties {soundWaveAssetPtr=/Game/Sounds/Fort_Weapons/AutoRifle/DrumGun/Drumgun_Fire_MediumNew_01}}
 {id=WavePlayer_21, properties {soundWaveAssetPtr=/Game/Sounds/Fort_Weapons/AutoRifle/DrumGun/Drumgun_Fire_MediumNew_02}}
 {id=WavePlayer_22, properties {soundWaveAssetPtr=/Game/Sounds/Fort_Weapons/AutoRifle/DrumGun/Drumgun_Fire_MediumNew_03}}
 {id=WavePlayer_23, properties {soundWaveAssetPtr=/Game/Sounds/Fort_Weapons/AutoRifle/DrumGun/Drumgun_Fire_DistantNew_01}}
 {id=WavePlayer_4, properties {soundWaveAssetPtr=&/RapidFireSMG_Fire_Close_03}}
 {id=WavePlayer_5, properties {soundWaveAssetPtr=&/RapidFireSMG_Fire_Close_04}}
]
properties {
 bDebug=true
 maxDistance=27500
 volumeMultiplier=1
 concurrencySet [&Weapon_Gunfire_Local_Conc, &Weapon_Gunfire_Global_Conc]
}
